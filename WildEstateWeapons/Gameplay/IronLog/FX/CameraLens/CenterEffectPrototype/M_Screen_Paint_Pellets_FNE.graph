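# Translucent, two-sided Niagara sprite material for paint pellet splats, apparently a prototype screen-space lens effect for the IronLog weapon.
# Particle color feeds an emissive adjustment function, while particle alpha is multiplied by a mask sharpened from a splash texture.
# Center-scaled Perlin noise, offset by the Dissolve dynamic parameter and contrast-boosted, erodes that mask.
# A random splash/goo-chunk blend and the Melt parameters are present but feed nothing shown.
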Material /WildEstateWeapons/Gameplay/IronLog/FX/CameraLens/CenterEffectPrototype/M_Screen_Paint_Pellets_FNE
graph {
 properties {
  bAutomaticallySetUsageInEditor=false
  bUsedWithNiagaraSprites=true
  blendMode=BLEND_Translucent
  twoSided=true
  usedShadingModels=Default Lit
 }
 nodes [
  n1 = TextureSample() {texture=/Game/Effects/Fort_Effects/Textures/Liquid/T_FX_Splash_01}
  n2 = ParticleRandom()
  LinearInterpolate(
   A=@n1.R
   B=TextureSample() {texture=/Game/Effects/Fort_Effects/Textures/Liquid/T_FX_GooChunks_01}.R
   Alpha=Floor(None=Multiply(A=@n2) {constB=2})
  )
  n3 = ParticleColor()
  MaterialFunctionCall(Emissive=@n3.RGB) {materialFunction=&MF_FX_EmissiveAdjustments}
  n4 = Saturate(None=Power(Base=OneMinus(None=@n1.R)) {constExponent=10})
  n5 = MaterialFunctionCall(Texture Scale=AppendVector(A=ScalarParameter() {defaultValue=1, parameterName=UScale}, B=ScalarParameter() {defaultValue=1, parameterName=VScale})) {materialFunction=&ScaleUVsByCenter}
  n6 = DynamicParameter() {defaultValue=<rgba 0.1 1 1 1>, paramNames [Melt Amount, Dip Tex Move amount, Melt Show, Dissolve]}
  n7 = Saturate(None=Add(A=TextureSample(UVs=@n5.UVs) {texture=/Game/Global/Textures/Default/Noise/3xPerlin_M}.R, B=LinearInterpolate(Alpha=@n6.Dissolve) {constA=-1}))
  Multiply(
   A=Multiply(A=@n3.A, B=OneMinus(None=@n4))
   B=Multiply(A=MaterialFunctionCall(In=@n7) {materialFunction=&CheapContrast}.Result)
  )
 ]
}
